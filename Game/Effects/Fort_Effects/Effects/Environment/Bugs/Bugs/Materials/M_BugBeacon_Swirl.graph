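# Unlit masked material for the bug beacon swirl, drawing flipbook billboards of a baked Niagara swarm atlas.
# Vertex color varies animation timing and speed, sprite scale and rotation, and culls cards by threshold; baked normals against the sun direction add fake lighting.
# World offset combines camera-facing sprites, slow rotation about the object pivot and scrolling curl noise.
# A camera-distance sphere mask collapses the sprites up close, so the swirl reads at long range.
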Material /Game/Effects/Fort_Effects/Effects/Environment/Bugs/Bugs/Materials/M_BugBeacon_Swirl
graph {
 properties {bAutomaticallySetUsageInEditor=false, blendMode=BLEND_Masked, shadingModel=MSM_Unlit, twoSided=true, usedShadingModels=Unlit}
 nodes [
  n1 = Time()
  n2 = VertexColor()
  n3 = LinearInterpolate(A=ScalarParameter() {defaultValue=1, parameterName=Time Scale}, Alpha=@n2.B) {constB=1.2}
  n4 = MaterialFunctionCall!(
   U_Images=ScalarParameter() {defaultValue=8, parameterName=X}
   V_Images=ScalarParameter() {defaultValue=8, parameterName=Y}
   SubUV Texture RGB
   Time Speed=LinearInterpolate(B=ScalarParameter() {defaultValue=2, parameterName="Anim Speed\\"}, Alpha=VertexColor().B) {constA=3}
   Manual Time Control=StaticBoolParameter() {defaultValue=true, parameterName="Manual Time Control (SB)"}
   Time=Add(A=Multiply(A=@n1, B=@n3), B=@n2.G)
   UVs
  ) {materialFunction=/Game/Effects/Fort_Effects/Materials/_Functions/MF_SubUV}
  n5 = TextureSample(UVs=@n4.UVs) {texture=/Game/Effects/Fort_Effects/Effects/Environment/Bugs/Bugs/Systems/NS_Bugs_SpawnerSwarm1_BakedAtlas_Texture2D_0}
  n6 = ScalarParameter() {defaultValue=2, parameterName="Albedo Gain (Med-Epic)"}
  n7 = ScalarParameter() {defaultValue=2, parameterName=Albedo Gain}
  n8 = TextureSample(UVs=@n4.UVs) {texture=/Game/Effects/Fort_Effects/Effects/Environment/Bugs/Bugs/Systems/NS_Bugs_SpawnerSwarm1_BakedAtlas_Texture2D_0_N}
  n9 = Transform(None=SkyAtmosphereLightDirection()) {transformSourceType=TRANSFORMSOURCE_World, transformType=TRANSFORM_Tangent}
  n10 = LinearInterpolate(A=Multiply(A=@n5.RGB) {constB=0.25}, B=@n5.RGB, Alpha=Saturate(None=DotProduct(A=@n8.RGB, B=@n9)))
  n11 = ScalarParameter() {defaultValue=6, parameterName=Lighting Boost}
  n12 = CollectionParameter() {collection=&FortniteMaterialParameters, parameterName=SunLightColor}
  Add(
   A=Multiply(
    A=@n5.RGB
    B=MaterialFunctionCall!("Xb1,PS4"=@n6, "PC-High&&Epic,XSX,PS5"=@n6, "PC-Perf,Android"=@n7, Switch=@n6, "PC-Low(NotPerf)&&Medium"=@n6) {materialFunction=/Game/Effects/Niagara/MasterMaterials/Functions/Common/PlatformSwitch_3D}.Result
   )
   B=Multiply(A=Multiply(A=@n10, B=@n11), B=ComponentMask(None=@n12) {b=true, g=true, r=true})
  )
  Constant()
  n13 = Multiply(A=@n5.A, B=ScalarParameter() {defaultValue=1, parameterName=Opacity Gain})
  n14 = SphereMask(
   A=TextureCoordinate()
   B=ScalarParameter() {defaultValue=0.5, parameterName=UV Mask Center}
   Radius=ScalarParameter() {defaultValue=0.6193, parameterName=UV Mask Radius}
   Hardness=ScalarParameter() {defaultValue=1, parameterName=UV Mask Hardness}
  )
  n15 = NamedRerouteUsage()
  Multiply(A=NamedRerouteUsage(), B=Multiply(A=Multiply(A=@n13, B=@n14), B=OneMinus(None=@n15)))
  NamedRerouteDeclaration(None=Step(Y=VertexColor().R, X=ScalarParameter() {defaultValue=0.45, parameterName=Step 1})) {name=Step, nodeColor=<rgba 0.5234 0 1 1>}
  n16 = ScalarParameter() {defaultValue=100, parameterName=Scale Min}
  n17 = ScalarParameter() {defaultValue=225, parameterName=Scale Max}
  n18 = VertexColor()
  n19 = NamedRerouteDeclaration(None=SphereMask(
    A=WorldPosition().XYZ
    B=CameraPositionWS()
    Radius=ScalarParameter() {defaultValue=25000, parameterName=Fade Distance}
    Hardness=ScalarParameter() {parameterName=Fade Hardness}
   )) {name=Sphere Mask, nodeColor=<rgba 0.6875 1 0 1>}
  n20 = MaterialFunctionCall!(
   Center Location=WorldPosition().XYZ
   0-1 UVs=TextureCoordinate()
   XY scale=Multiply(A=LinearInterpolate(A=@n16, B=@n17, Alpha=@n18.G), B=OneMinus(None=@n19))
   OverridePivot
   "Normals (Optional)"
  ) {materialFunction=&Sprite}
  n21 = Time()
  n22 = ScalarParameter() {defaultValue=0.015, parameterName=Rotation Speed}
  n23 = VertexColor()
  n24 = RotateAboutAxis(
   NormalizedRotationAxis=VectorParameter() {defaultValue=<rgba 0 0 1 0>}.RGB
   RotationAngle=Add(A=Multiply(A=@n21, B=@n22), B=LinearInterpolate(Alpha=@n23.G) {constA=1, constB=2})
   PivotPoint=ObjectPositionWS()
   Position=WorldPosition().XYZ
  )
  n25 = ScalarParameter() {defaultValue=350, parameterName=Curl Scale}
  n26 = WorldPosition()
  n27 = Multiply(A=Time(), B=VectorParameter() {parameterName=Curl Pain Dir, defaultValue=<rgba 0 0 150 0>}.RGB)
  n28 = VectorNoise(World Position=Divide(A=Add(A=@n26.XYZ, B=@n27), B=ScalarParameter() {defaultValue=1024, parameterName=Noise Size})) {noiseFunction=VNF_CurlALU}
  Reroute(None=Add(A=Add(A=@n20.World position offset, B=@n24), B=Multiply(A=@n25, B=@n28)))
 ]
}
