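# Niagara environment effect that circles a swarm of mesh bugs overhead.
# The active Minimal emitter burst-spawns bugs that orbit randomized centers on tilted axes, using curl noise position offsets instead of force solving and facing their travel direction.
# The disabled GPU Bugs emitter is a heavier variant with distance-field surface placement, camera-relative point force, vortex, noise and spring-back forces.
# Particles outlive their lifetime, with removal driven by a scratch kill condition.
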NiagaraSystem /Game/Effects/Fort_Effects/Effects/Environment/Bugs/Bugs/Systems/NS_Bugs_OverheadSwarm
dependencies {
 usedDataInterfaces [&NiagaraDataInterfaceCamera, &NiagaraDataInterfaceDebugDraw, &NiagaraDataInterfaceParticleRead]
 usedDynamicInputs []
 usedModules [&SystemState]
}
emitters {
 Bugs {
  bEnabled=false
  simTarget=GPUComputeSim
  emitterUpdateScript [
   &EmitterState(Life Cycle Mode=::NewEnumerator1, Loop Behavior=::NewEnumerator1, Loop Duration Mode=::NewEnumerator1, UseLoopDelay=false)
   &SpawnBurst_Instantaneous(Spawn Count=64, Use Spawn Probability=false, UseLoopCountLimit=false)
  ]
  particleSpawnScript [
   &InitializeParticle(
    Color=<rgba 0.9937 0 1 0.01>
    Lifetime=5
    Mesh Scale Mode=::NewEnumerator1
    Mesh Uniform Scale Max=0.15
    Mesh Uniform Scale Min=0.1
    Sprite Size Mode=::NewEnumerator3
    Uniform Sprite Size=175
    UsePositionOffset=false
    Write Material Random=true
   )
   &ShapeLocation(
    Handle Radius=500
    Large Radius=&RandomRangeFloat(Minimum=2500, Maximum=3500)
    MaskSpawnGroup=false
    Offset=<0 0 500>
    Offset Mode=::NewEnumerator0
    Override Seed=false
    Shape Primitive=::NewEnumerator3
    UseOwnerScale=true
   )
   /Niagara/Modules/Math/MoveToNearestDistanceFieldSurface_GPU(Kill by Distance=2500, Radius=1, bKillByDistance=true)
   &InitialMeshOrientation(
    Mesh Axis To Orient=<0 0 -1>
    Mesh Orientation Mode=::NewEnumerator2
    Orientation Vector=$Output.MoveToNearestDistanceFieldSurface_GPU.VectorToNearestSurfaceNormal
    Use Rotation Vector=false
   )
   /SetVariables(
    Particles.Awake=true
    Particles.BirdMeshScale:Vector3f=&VectorFromFloat(Value=&RandomRangeFloat(Minimum=0.1, Maximum=0.125))
   ) {isSetParameters=true}
  ]
  particleUpdateScript [
   &ParticleState(Kill Particles When Lifetime Has Elapsed=false, Let Infinitely Lived Particles Die When Emitter Deactivates=true)
   /SetVariables(Particles.Player Close:NiagaraBool=&SetBoolByFloatComparison(
     A=&VectorLength(Vector=&SubtractVectorFromPosition(
       Position=&ReturnCameraPosition(Coordinate Space:ENiagaraCoordinateSpace=::Simulation, Camera Query:NiagaraDataInterfaceCamera=$System.Camera Query)
       Vector=Linked() {linkedVariable {name=Particles.Position, type=NiagaraPosition}}
      ))
     B=3500
     Comparison Type=::NewEnumerator6
    )) {isSetParameters=true}
   &PointForce(
    Force Falloff Distance:NiagaraFloat=&Power(a=&RandomRangeFloat(Minimum=500, Maximum=5000), b=8)
    Force Origin=&AddVectorToPosition(
     Position=&ReturnCameraPosition(Coordinate Space:ENiagaraCoordinateSpace=::World, Camera Query:NiagaraDataInterfaceCamera=DataInterface())
     Vector=<0 0 -250>
    )
    Force Strength=1500
    Module.Use Falloff=true
    Module.UseLinearForce=false
   )
   &PointForce(Force Falloff Distance=1000, Force Strength=3500, Module.Use Falloff=true, Module.UseLinearForce=false) {module=PointForce001}
   &LinearForce(
    Coordinate Space:ENiagaraCoordinateSpace=::World
    Force=&Multiply_VectorByFloat(
     Vector=&Subtract_Vector(
      A=Linked() {linkedVariable {name=Particles.Initial.Position, type=NiagaraPosition}}
      B=Linked() {linkedVariable {name=Particles.Position, type=NiagaraPosition}}
     )
     Float=&RandomRangeFloat(Minimum=1, Maximum=2)
    )
   )
   &VortexForce(Origin Pull Amount=500, UseOriginOffset=false, Vortex Force Amount=75000)
   &CurlNoiseForce(
    Bias Noise=false
    Noise Frequency=&RandomRangeFloat(Minimum=0.25, Maximum=1)
    Noise Strength=75000
    Pan Noise=true
    Pan Noise Field=<0 0 0.25>
   )
   &Drag(Use Linear Drag=true, Use Rotational Drag=true)
   &SolveForcesAndVelocity(Acceleration Limit=1500, Clamp Velocity=true, Limit Acceleration=true)
   &UpdateMeshOrientation(
    "Foward Vector (Mesh Space)":Vector3f=<0 -1 0>
    Initial Forward Vector:Vector3f=<0 1 0>
    Orientation Method=::NewEnumerator3
    "Right Vector (Mesh Space)":Vector3f=<1 0 0>
    Upright Roll Correction Speed=3
   )
   /SetVariables(Particles.ParticleUpdateForce:Vector3f=$Transient.PhysicsForce) {isSetParameters=true}
   &ScaleMeshSize(Scale Factor=&VectorFromFloat(Value=&Lerp_Float(A=1, B=3, Alpha=$Emitter.Scalability.DistanceFraction)))
   /NiagaraFluids/Modules/Grid3D/NeighborGrid3D_Visualize(
    GridToVisualize:NiagaraDataInterfaceNeighborGrid3D=$Emitter.InitializeNeighborGrid.NeighborGrid3D
    UnitToWorld:NiagaraMatrix=$Emitter.GridUnitToWorld
   ) {enabled=false}
   &KillParticles(Kill Particles=&self:ScratchDynamicInput_01())
  ]
  rendererClasses [&NiagaraMeshRendererProperties]
  renderers [{rendererClass=&NiagaraMeshRendererProperties, rendererIndex=0}]
 }
 Minimal {
  emitterUpdateScript [
   &EmitterState(Life Cycle Mode=::NewEnumerator1, Loop Behavior=::NewEnumerator1, Loop Duration Mode=::NewEnumerator1, UseLoopDelay=false)
   &SpawnBurst_Instantaneous(Spawn Count=50, Use Spawn Probability=false, UseLoopCountLimit=false)
  ]
  particleSpawnScript [
   &InitializeParticle(
    Color=<rgba 1 0 0 1>
    Lifetime=5
    Mesh Scale Mode=::NewEnumerator1
    Mesh Uniform Scale Max=0.25
    Mesh Uniform Scale Min=0.125
    UsePositionOffset=false
    Write Material Random=true
   )
   &ShapeLocation(Handle Radius=1500, Large Radius=2500, MaskSpawnGroup=false, Override Seed=false, Shape Primitive=::NewEnumerator3, UseOwnerScale=true)
  ]
  particleUpdateScript [
   &ParticleState(Kill Particles When Lifetime Has Elapsed=false, Let Infinitely Lived Particles Die When Emitter Deactivates=true)
   &CurlNoiseForce(Bias Noise=false, Noise Frequency=1, Noise Quality / Cost=::NewEnumerator0, Noise Strength=75, Pan Noise=true, Pan Noise Field=<0 0 0.1>)
   &RotateAroundPoint(
    Derive the Rotation Axis=false
    Radius:NiagaraFloat=&RandomRangeFloat(Minimum=1500, Maximum=3500)
    Rotation Axis:Vector3f=&RandomRangeVector(Minimum=<-0.5 -0.5 1>, Maximum=<0.5 0.5 1>)
    Rotation Center:NiagaraPosition=&AddVectorToPosition(
     Position=/Niagara/DynamicInputs/Helpers/SimulationPosition()
     Vector=&RandomRangeVector(Minimum=<0 0 -1000>, Maximum=<1 1 1500>)
    )
    Rotation Phase:NiagaraFloat=&RandomRangeFloat(Minimum=0, Maximum=1)
    Rotation Rate:NiagaraFloat=&RandomRangeFloat(Minimum=0.05, Maximum=0.075)
   )
   /SetVariables(Particles.Position=&AddVectorToPosition(Position=$Particles.Position, Vector=$Output.CurlNoiseForce.CurlNoiseForce)) {isSetParameters=true}
   /SetVariables(Particles.MeshOrientation=&MakeQuaternion(
     Quaternion From=3
     Axis=<1 0 0>
     Angle=0
     From=<0 -1 0>
     To=$Output.RotateAroundPoint.LocalForwardVector
     Angle Type=::NewEnumerator0
     Normalize Vectors=true
    )) {isSetParameters=true}
   &KillParticles(Kill Particles=&self:ScratchDynamicInput_01())
  ]
  rendererClasses [&NiagaraMeshRendererProperties]
  renderers [
   {rendererClass=&NiagaraMeshRendererProperties, rendererIndex=0}
   {rendererClass=&NiagaraMeshRendererProperties, rendererIndex=1}
  ]
 }
}
summary {
 systemName=NS_Bugs_OverheadSwarm
 emitters [
  {bEnabled=false, emitterName=Bugs, simTarget=GPUComputeSim, rendererClasses [&NiagaraMeshRendererProperties]}
  {emitterName=Minimal, rendererClasses [&NiagaraMeshRendererProperties]}
 ]
 userVariables []
}
